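# Translucent unlit material for security glass that shifts from Color_A (blue) to Color_B (red) when damaged or through the State parameter.
# The length of the destructible-damage function's DamageAnimWPO output becomes the color blend alpha, maxed with State.
# A tiled texture and DepthFade intersection glow form one mask driving glow brightness and opacity.
# A desaturated, contrast-adjusted lat-long texture adds fake reflections, and EyeAdaptationInverse compensates the combined color for exposure.
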Material /Game/Environments/TechArt/SecurityGlass/M_SecurityGlass
graph {
 properties {bAutomaticallySetUsageInEditor=false, blendMode=BLEND_Translucent, shadingModel=MSM_Unlit, usedShadingModels=Unlit}
 nodes [
  Add()
  n1 = StaticBool()
  n2 = ScalarParameter() {defaultValue=0, parameterName=Custom Damage Radius}
  n3 = Divide(A=Length(None=MaterialFunctionCall(UseAlternateHitGradientOutput=@n1, WSDamageRadius=@n2) {materialFunction=&MF_FN_DestructibleDamage}.DamageAnimWPO) {is=class MaterialExpressionLength}) {constB=0.01}
  n4 = LinearInterpolate(
   A=VectorParameter() {parameterName=Color_A, defaultValue=<rgba 0 0.2942 1 1>}.RGB
   B=VectorParameter() {parameterName=Color_B, defaultValue=<rgba 1 0.0365 0.0365 1>}.RGB
   Alpha=Max(A=Saturate(None=@n3), B=ScalarParameter() {parameterName=State})
  )
  n5 = TextureCoordinate()
  n6 = ScalarParameter() {defaultValue=1.5, parameterName=Tiling}
  n7 = DepthFade(FadeDistance=ScalarParameter() {defaultValue=5, parameterName=DepthFade Glow})
  n8 = Max(
   A=TextureSample(UVs=Multiply(A=@n5, B=@n6)) {texture=&/SecurityGlass_M}.R
   B=Multiply(A=ScalarParameter() {defaultValue=2, parameterName=EdgeGlowBrightness}, B=OneMinus(None=@n7))
  )
  n9 = LinearInterpolate(A=ScalarParameter() {defaultValue=1, parameterName=Glow_B}, B=ScalarParameter() {defaultValue=10.3279, parameterName=Glow_A}, Alpha=@n8)
  n10 = MaterialFunctionCall!(Vector=ReflectionVectorWS()) {materialFunction=/Engine/Functions/Engine_MaterialFunctions01/Coordinates/LongLatToUV}
  n11 = MaterialFunctionCall!(
   In=Desaturation(
    None=TextureSample(UVs=@n10.Result) {texture=/Game/Environments/Asteria/Sets/NeonCity/Techart/Textures/T_FakeInterior_01_LongLat}.RGB
    Fraction=ScalarParameter() {defaultValue=0.5, parameterName=Reflection_Desaturation}
   )
   Contrast=ScalarParameter() {defaultValue=-0.25, parameterName=Reflection_Contrast}
  ) {materialFunction=/Engine/Functions/Engine_MaterialFunctions01/ImageAdjustment/CheapContrast_RGB}
  n12 = ScalarParameter() {defaultValue=0.7, parameterName=Reflection_Intensity}
  EyeAdaptationInverse(LightValueInput=Add(A=Multiply(A=@n4, B=@n9), B=Multiply(A=@n11.Result, B=@n12)))
  LinearInterpolate(
   A=ScalarParameter() {defaultValue=0.168, parameterName=OpacityB}
   B=ScalarParameter() {defaultValue=0.324, parameterName=OpacityA}
   Alpha=@n8
  )
 ]
}
